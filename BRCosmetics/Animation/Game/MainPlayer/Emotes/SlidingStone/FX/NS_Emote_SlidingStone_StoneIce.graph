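# Niagara system for the Sliding Stone emote that lays ice decals around the moving owner, presumably the stone's trail on the ice.
# Leader and Follower decal emitters ramp up spawning at different ages, with Leader offset further and shorter-lived.
# Owner velocity drives a smoothed, speed-gated spawn offset.
# Decals get random orientation and size, shrink via User.Follower, and animate through red-channel and alpha curves; a front-end preview flag picks the wet color.
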NiagaraSystem /BRCosmetics/Animation/Game/MainPlayer/Emotes/SlidingStone/FX/NS_Emote_SlidingStone_StoneIce
dependencies {
 usedDataInterfaces [&NiagaraDataInterfaceCurve, &NiagaraDataInterfaceArrayInt32, &NiagaraDataInterfaceArrayColor]
 usedDynamicInputs []
 usedModules [
  &self:SystemSpawnScript.SystemScriptSource.SystemScriptGraph.NiagaraNodeAssignment_0.SetVariables_#1
  &SystemState
 ]
}
emitters {
 Follower {
  emitterSpawnScript [/SetVariables(Emitter.WetColor:LinearColor=&MakeCustomLinearColorFromBool(Bool=$User.bIsFrontEndPreview, True Color=<rgba 0.0892 0.3666 0.6 1>, False Color=<rgba 0.2002 0.3087 0.4 1>)) {isSetParameters=true, module=SetVariables_#2}]
  emitterUpdateScript [
   &EmitterState(Life Cycle Mode=::NewEnumerator1, Loop Behavior=::NewEnumerator1, Loop Duration Mode=::NewEnumerator1, UseLoopDelay=false)
   &SpawnRate(
    SpawnRate=&FloatFromCurve(FloatCurve=DataInterface() {propertyValues {Curve {keys [<at 0.0725 0>,<at 0.1725 1>]}}}, CurveIndex=$Emitter.Age, Scale Curve=10)
    Use Spawn Probability=false
   )
   /SetVariables(Emitter.PushVector:Vector3f=&MakeCustomVectorFromBool(
     Bool=&SetBoolByFloatComparison(A=&VectorLength(Vector=$Engine.Owner.Velocity), B=200)
     True Vector=<0 0 1>
     False Vector=&Multiply_Vector(
      A=&TransformVector(Vector=$Engine.Owner.Velocity, Source Space=::World, Destination Space=::World)
      B=&Multiply_Vector(
       A=<1 1 1>
       B=&VectorFromFloat(Value=/Niagara/DynamicInputs/Lerp/SmoothLerpOverTime_Float(
         Convergence Rate:NiagaraFloat=2
         Smooth Value:NiagaraFloat=0
         Target Value:NiagaraFloat=&MakeCustomFloatFromBool(
          Bool=&SetBoolByFloatComparison(A=&VectorLength(Vector=$Engine.Owner.Velocity), B=50)
          True Float=1
          False Float=0
         )
        ))
      )
     )
    )) {isSetParameters=true, module=SetVariables_#3}
  ]
  particleSpawnScript [
   &InitializeParticle(
    Color=&SelectLinearColorFromArray(
     Array Sampling Mode=::NewEnumerator0
     Direct Array Index=?
     Color Selection Array=DataInterface() {propertyValues {ColorData [<rgba 1 0.9819 0.71 1>, <rgba 1 0.9623 0.4532 1>]}}
     Random Seed=0
    )
    Lifetime=4
    Mesh Scale Mode=::NewEnumerator3
    Mesh Uniform Scale=10
    Position Offset=$Emitter.PushVector
    Position Offset Coordinate Space=::World
    Sprite Rotation Mode=::NewEnumerator2
    Sprite Size Mode=::NewEnumerator1
    Uniform Sprite Size Max=30
    Uniform Sprite Size Min=50
    UsePositionOffset=true
    Write Material Random=true
   )
   /SetVariables(
    Particles.DecalOrientation:Quat4f=/Game/Effects/Niagara/Modules/MergeEmitters/MergeHelper_QuatSwitch(
     0:Quat4f=<0 1 0 1>
     1:Quat4f=<0.2 1 0 1>
     2:Quat4f=<0.4 1 0 1>
     3:Quat4f=<0.6 1 0 1>
     4:Quat4f=<0.8 1 0 1>
     5:Quat4f=?
     6:Quat4f=?
     7:Quat4f=?
     8:Quat4f=?
     9:Quat4f=?
     Index:NiagaraInt32=&SelectIntFromArray(Direct Array Index=0, Int Selection Array=DataInterface() {propertyValues {IntData [0,1,2,3,4]}})
     Num:MergeHelper_SwitcherEnum=::NewEnumerator3
    )
    Particles.DecalSize:Vector3f=&Lerp_Vector(
     A=&RandomRangeVector(Minimum=<200 72.5 72.5>, Maximum=<200 145 145>)
     B=&Multiply_Vector(
      A=&RandomRangeVector(Minimum=<200 72.5 72.5>, Maximum=<200 145 145>)
      B=&VectorFromFloat(Value=0.45)
     )
     Alpha=$User.Follower
    )
   ) {isSetParameters=true, module=SetVariables_#4}
   &ShapeLocation(
    Box Size=<0 0 10>
    MaskSpawnGroup=false
    Non Uniform Scale=&MaskVector(Mask Source=0, Mask Parameter=?, Mask=1, Masked Vector=<0 0 0>, Passthrough Vector=<1 1 1>)
    Override Seed=false
    Shape Primitive=::NewEnumerator2
    Surface Only=false
    UseOwnerScale=true
   )
  ]
  particleUpdateScript [
   &ParticleState()
   /SetVariables(Particles.DecalColor:LinearColor=&MakeLinearColorFromVectorAndFloat(
     "Vector (RGB)"=&MakeVector(
      X=&FloatFromCurve(
       FloatCurve=DataInterface() {propertyValues {Curve {keys [{arriveTangent=2.6346, interpMode=RCIM_Cubic, leaveTangent=2.6346, tangentMode=RCTM_User, time=0, value=-1}, <at 1 0>]}}}
       Scale Curve=&RandomRangeFloat(Minimum=0.5, Maximum=0.8)
      )
      Y=0
      Z=0
     )
     "Float (Alpha)"=&FloatFromCurve(FloatCurve=DataInterface() {propertyValues {Curve {
         defaultValue=1
         keys [
          {arriveTangent=5.9515, interpMode=RCIM_Cubic, leaveTangent=5.9515, tangentMode=RCTM_User, time=0, value=0}
          {interpMode=RCIM_Cubic, tangentMode=RCTM_User, time=0.104, value=1}
          {arriveTangentWeight=0.2201, interpMode=RCIM_Cubic, leaveTangentWeight=0.2332, tangentMode=RCTM_User, time=0.5259, value=1}
          {arriveTangent=0.086, interpMode=RCIM_Cubic, leaveTangent=0.086, tangentMode=RCTM_User, time=1, value=0}
         ]
        }}})
    )) {isSetParameters=true, module=SetVariables_#5}
   &SolveForcesAndVelocity(Clamp Velocity=false, Limit Acceleration=false)
  ]
  rendererClasses [&NiagaraDecalRendererProperties]
  renderers [{rendererClass=&NiagaraDecalRendererProperties, rendererIndex=0}]
 }
 Leader {
  emitterSpawnScript [/SetVariables(Emitter.WetColor:LinearColor=&MakeCustomLinearColorFromBool(Bool=$User.bIsFrontEndPreview, True Color=<rgba 0.0892 0.3666 0.6 1>, False Color=<rgba 0.2002 0.3087 0.4 1>)) {isSetParameters=true, module=SetVariables_#2}]
  emitterUpdateScript [
   &EmitterState(Life Cycle Mode=::NewEnumerator1, Loop Behavior=::NewEnumerator1, Loop Duration Mode=::NewEnumerator1, UseLoopDelay=false)
   &SpawnRate(
    SpawnRate=&FloatFromCurve(FloatCurve=DataInterface() {propertyValues {Curve {keys [<at 0.6083 0>,<at 0.7083 1>]}}}, CurveIndex=$Emitter.Age, Scale Curve=12)
    Use Spawn Probability=false
   )
   /SetVariables(Emitter.PushVector:Vector3f=&MakeCustomVectorFromBool(
     Bool=&SetBoolByFloatComparison(A=&VectorLength(Vector=$Engine.Owner.Velocity), B=200)
     True Vector=<0 0 1>
     False Vector=&Multiply_Vector(
      A=&TransformVector(Vector=$Engine.Owner.Velocity, Source Space=::World, Destination Space=::World)
      B=&Multiply_Vector(
       A=<1.75 1.75 1.75>
       B=&VectorFromFloat(Value=&SmoothLerpOverTime_Float(
         Convergence Rate:NiagaraFloat=2
         Smooth Value:NiagaraFloat=0
         Target Value:NiagaraFloat=&MakeCustomFloatFromBool(
          Bool=&SetBoolByFloatComparison(A=&VectorLength(Vector=$Engine.Owner.Velocity), B=50)
          True Float=1
          False Float=0
         )
        ))
      )
     )
    )) {isSetParameters=true, module=SetVariables_#3}
  ]
  particleSpawnScript [
   &InitializeParticle(
    Color=<rgba 1 0.9576 0.9894 1>
    Lifetime=1.8
    Mesh Scale Mode=::NewEnumerator3
    Mesh Uniform Scale=10
    Position Offset=$Emitter.PushVector
    Position Offset Coordinate Space=::World
    Sprite Rotation Mode=::NewEnumerator2
    Sprite Size Mode=::NewEnumerator1
    Uniform Sprite Size Max=30
    Uniform Sprite Size Min=50
    UsePositionOffset=true
    Write Material Random=true
   )
   /SetVariables(
    Particles.DecalOrientation:Quat4f=&MergeHelper_QuatSwitch(
     0:Quat4f=<0 1 0 1>
     1:Quat4f=<0.2 1 0 1>
     2:Quat4f=<0.4 1 0 1>
     3:Quat4f=<0.6 1 0 1>
     4:Quat4f=<0.8 1 0 1>
     5:Quat4f=?
     6:Quat4f=?
     7:Quat4f=?
     8:Quat4f=?
     9:Quat4f=?
     Index:NiagaraInt32=&SelectIntFromArray(Direct Array Index=0, Int Selection Array=DataInterface() {propertyValues {IntData [0,1,2,3,4]}})
     Num:MergeHelper_SwitcherEnum=::NewEnumerator3
    )
    Particles.DecalSize:Vector3f=&Lerp_Vector(
     A=&Multiply_Vector(
      A=&RandomRangeVector(Minimum=<200 72.5 72.5>, Maximum=<200 145 145>)
      B=&VectorFromFloat(Value=1.1)
     )
     B=&Multiply_Vector(
      A=&RandomRangeVector(Minimum=<200 72.5 72.5>, Maximum=<200 145 145>)
      B=&VectorFromFloat(Value=0.6)
     )
     Alpha=$User.Follower
    )
   ) {isSetParameters=true, module=SetVariables_#4}
   &ShapeLocation(
    Box Size=<0 0 10>
    MaskSpawnGroup=false
    Non Uniform Scale=&MaskVector(Mask Source=0, Mask Parameter=?, Mask=1, Masked Vector=<0 0 0>, Passthrough Vector=<1 1 1>)
    Override Seed=false
    Shape Primitive=::NewEnumerator2
    Surface Only=false
    UseOwnerScale=true
   )
  ]
  particleUpdateScript [
   &ParticleState()
   /SetVariables(Particles.DecalColor:LinearColor=&MakeLinearColorFromVectorAndFloat(
     "Vector (RGB)"=&MakeVector(
      X=&FloatFromCurve(
       FloatCurve=DataInterface() {propertyValues {Curve {keys [<at 0.348 -0.55>,<at 1 0>]}}}
       Scale Curve=&RandomRangeFloat(Minimum=1.2, Maximum=0.9)
      )
      Y=0
      Z=0
     )
     "Float (Alpha)"=&FloatFromCurve(FloatCurve=DataInterface() {propertyValues {Curve {
         defaultValue=1
         keys [
          {arriveTangent=5.9515, interpMode=RCIM_Cubic, leaveTangent=5.9515, tangentMode=RCTM_User, time=0, value=0}
          {interpMode=RCIM_Cubic, tangentMode=RCTM_User, time=0.5, value=1}
          {arriveTangentWeight=0.2201, interpMode=RCIM_Cubic, leaveTangentWeight=0.2332, tangentMode=RCTM_User, time=0.9544, value=1}
          {arriveTangent=0.086, interpMode=RCIM_Cubic, leaveTangent=0.086, tangentMode=RCTM_User, time=1, value=0}
         ]
        }}})
    )) {isSetParameters=true, module=SetVariables_#5}
   &SolveForcesAndVelocity(Clamp Velocity=false, Limit Acceleration=false)
  ]
  rendererClasses [&NiagaraDecalRendererProperties]
  renderers [{rendererClass=&NiagaraDecalRendererProperties, rendererIndex=0}]
 }
}
summary {
 systemName=NS_Emote_SlidingStone_StoneIce
 emitters [
  {emitterName=Leader, rendererClasses [&NiagaraDecalRendererProperties]}
  {emitterName=Follower, rendererClasses [&NiagaraDecalRendererProperties]}
 ]
 userVariables [
  User.bIsFrontEndPreview {type=NiagaraBool}
  User.Follower {type=NiagaraFloat, defaultValue=0, description=""}
  User.OwnerLocalVelocity {type=Vector3f, description="", defaultValue {struct=Vector3f, value=<0 0 0>}}
 ]
}
